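# Post-process style AnimBP for the female medium Frogman body cosmetic, correcting secondary bones on top of the pose copied from its attached parent mesh.
# Weighted multi-bone constraints position eight buckle bones and the gun base from spine, clavicle, arm, pelvis and thigh bones.
# Each update maps thigh swing and left-thigh yaw to alphas that scale the sub-thigh bones and offset the gun base.
# Two constant sub-thigh scale nodes are unconnected.
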AnimBlueprint /BRCosmetics/Characters/Player/Female/Medium/Bodies/F_MED_Frogman/Meshes/F_MED_Frogman_AnimBP
target_skeleton=&/F_MED_Frogman_Skeleton
graphs [
 AnimGraph {nodes [
   n1 = "Transform (Modify) Bone - Bone: dyn_subthigh_l"(Alpha=1, Translation="0,0,0", Rotation="0,0,0", Scale="10,10,10") {
    scaleMode=BMM_Additive
    scaleSpace=BCS_BoneSpace
    translationMode=BMM_Additive
    translationSpace=BCS_BoneSpace
    alphaBoolBlend {blendOption=Linear}
    alphaScaleBiasClamp {interpSpeedDecreasing=9, interpSpeedIncreasing=9}
    boneToModify {boneName=dyn_subthigh_l}
    scale=<10 10 10>
   }
   "Transform (Modify) Bone - Bone: dyn_subthigh_r"(ComponentPose=@n1, Alpha=1, Translation="0,0,0", Rotation="0,0,0", Scale="2,2,2") {
    scaleMode=BMM_Additive
    scaleSpace=BCS_BoneSpace
    translationMode=BMM_Additive
    translationSpace=BCS_BoneSpace
    alphaBoolBlend {blendOption=Linear}
    alphaScaleBiasClamp {interpSpeedDecreasing=9, interpSpeedIncreasing=9}
    boneToModify {boneName=dyn_subthigh_r}
    scale=<2 2 2>
   }
   n2 = "Constraint - cosnt_buckle_topbk_l (PPP)"!(
    ComponentPose=Local To Component(LocalPose=Copy Pose From Mesh() {bUseAttachedParent=true})
    Alpha=1
    ConstraintWeights_0=0.8871
    ConstraintWeights_1:float=0.0965
    ConstraintWeights_2:float=0.0164
   ) -> (Pose) {
    alphaBoolBlend {blendOption=Linear}
    boneToModify {boneName=cosnt_buckle_topbk_l}
    constraintSetup [{targetBone {boneName=clavicle_l}}, {targetBone {boneName=spine_05}}, {targetBone {boneName=upperarm_twist_01_l}}]
    constraintWeights [1,1,1]
   }
   n3 = "Constraint - cosnt_buckle_topbk_r (PPPPP)"!(
    ComponentPose=@n2
    Alpha=1
    ConstraintWeights_0=0.7288
    ConstraintWeights_1:float=0.0418
    ConstraintWeights_2:float=0.0038
    ConstraintWeights_3:float=0.2218
    ConstraintWeights_4:float=0.0037
   ) -> (Pose) {
    alphaBoolBlend {blendOption=Linear}
    boneToModify {boneName=cosnt_buckle_topbk_r}
    constraintSetup [
     {targetBone {boneName=clavicle_r}}
     {targetBone {boneName=upperarm_twist_01_r}}
     {targetBone {boneName=spine_04}}
     {targetBone {boneName=spine_05}}
     {targetBone {boneName=clavicle_l}}
    ]
    constraintWeights [1,1,1,1,1]
   }
   n4 = "Constraint - cosnt_buckle_bottombk_l (PPPPPPPP)"!(
    ComponentPose=@n3
    Alpha=1
    ConstraintWeights_0=0.097
    ConstraintWeights_1:float=0.038
    ConstraintWeights_2:float=0.1123
    ConstraintWeights_3:float=0.1901
    ConstraintWeights_4:float=0.5022
    ConstraintWeights_5:float=0.0575
    ConstraintWeights_6:float=0.0023
    ConstraintWeights_7:float=0.0006
   ) -> (Pose) {
    alphaBoolBlend {blendOption=Linear}
    boneToModify {boneName=cosnt_buckle_bottombk_l}
    constraintSetup [
     {targetBone {boneName=clavicle_l}}
     {targetBone {boneName=spine_01}}
     {targetBone {boneName=spine_02}}
     {targetBone {boneName=spine_03}}
     {targetBone {boneName=spine_04}}
     {targetBone {boneName=spine_05}}
     {targetBone {boneName=upperarm_twist_01_l}}
     {targetBone {boneName=clavicle_r}}
    ]
    constraintWeights [1,1,1,1,1,1,1,1]
   }
   n5 = "Constraint - cosnt_buckle_bottombk_r (PPPPPPPP)"!(
    ComponentPose=@n4
    Alpha=1
    ConstraintWeights_0=0.0001
    ConstraintWeights_1:float=0.0006
    ConstraintWeights_2:float=0.0364
    ConstraintWeights_3:float=0.1092
    ConstraintWeights_4:float=0.2242
    ConstraintWeights_5:float=0.5117
    ConstraintWeights_6:float=0.0403
    ConstraintWeights_7:float=0.0776
   ) -> (Pose) {
    alphaBoolBlend {blendOption=Linear}
    boneToModify {boneName=cosnt_buckle_bottombk_r}
    constraintSetup [
     {targetBone {boneName=clavicle_l}}
     {targetBone {boneName=upperarm_twist_01_r}}
     {targetBone {boneName=spine_01}}
     {targetBone {boneName=spine_02}}
     {targetBone {boneName=spine_03}}
     {targetBone {boneName=spine_04}}
     {targetBone {boneName=spine_05}}
     {targetBone {boneName=clavicle_r}}
    ]
    constraintWeights [1,1,1,1,1,1,1,1]
   }
   n6 = "Constraint - cosnt_buckle_bottomfrnt_r (PPPPP)"!(
    ComponentPose=@n5
    Alpha=1
    ConstraintWeights_0=0.0318
    ConstraintWeights_1:float=0.0741
    ConstraintWeights_2:float=0.1076
    ConstraintWeights_3:float=0.7823
    ConstraintWeights_4:float=0.0041
   ) -> (Pose) {
    alphaBoolBlend {blendOption=Linear}
    boneToModify {boneName=cosnt_buckle_bottomfrnt_r}
    constraintSetup [
     {targetBone {boneName=spine_01}}
     {targetBone {boneName=spine_02}}
     {targetBone {boneName=spine_03}}
     {targetBone {boneName=spine_04}}
     {targetBone {boneName=pelvis}}
    ]
    constraintWeights [1,1,1,1,1]
   }
   n7 = "Constraint - cosnt_buckle_bottomfrnt_l (PPPPP)"!(
    ComponentPose=@n6
    Alpha=1
    ConstraintWeights_0=0.0304
    ConstraintWeights_1:float=0.0695
    ConstraintWeights_2:float=0.0905
    ConstraintWeights_3:float=0.8059
    ConstraintWeights_4:float=0.0036
   ) -> (Pose) {
    alphaBoolBlend {blendOption=Linear}
    boneToModify {boneName=cosnt_buckle_bottomfrnt_l}
    constraintSetup [
     {targetBone {boneName=spine_01}}
     {targetBone {boneName=spine_02}}
     {targetBone {boneName=spine_03}}
     {targetBone {boneName=spine_04}}
     {targetBone {boneName=pelvis}}
    ]
    constraintWeights [1,1,1,1,1]
   }
   n8 = "Constraint - cosnt_buckle_topfrnt_r (PPPP)"!(
    ComponentPose=@n7
    Alpha=1
    ConstraintWeights_0=0.2962
    ConstraintWeights_1:float=0.0483
    ConstraintWeights_2:float=0.1568
    ConstraintWeights_3:float=0.4988
   ) -> (Pose) {
    alphaBoolBlend {blendOption=Linear}
    boneToModify {boneName=cosnt_buckle_topfrnt_r}
    constraintSetup [
     {targetBone {boneName=clavicle_r}}
     {targetBone {boneName=upperarm_twist_01_r}}
     {targetBone {boneName=spine_04}}
     {targetBone {boneName=spine_05}}
    ]
    constraintWeights [1,1,1,1]
   }
   n9 = "Constraint - cosnt_buckle_topfrnt_l (PPPP)"!(
    ComponentPose=@n8
    Alpha=1
    ConstraintWeights_0=0.3521
    ConstraintWeights_1:float=0.1242
    ConstraintWeights_2:float=0.4537
    ConstraintWeights_3:float=0.07
   ) -> (Pose) {
    alphaBoolBlend {blendOption=Linear}
    boneToModify {boneName=cosnt_buckle_topfrnt_l}
    constraintSetup [
     {targetBone {boneName=clavicle_l}}
     {targetBone {boneName=spine_04}}
     {targetBone {boneName=spine_05}}
     {targetBone {boneName=upperarm_twist_01_l}}
    ]
    constraintWeights [1,1,1,1]
   }
   n10 = "Transform (Modify) Bone - Bone: dyn_subthigh_l"(ComponentPose=@n9, Alpha=Get legSwingDelta_l() -> (legSwingDelta_l:double), Translation="0,0,0", Rotation="0,0,0", Scale="12,1,12") {
    scaleMode=BMM_Additive
    scaleSpace=BCS_BoneSpace
    translationMode=BMM_Additive
    translationSpace=BCS_BoneSpace
    alphaBoolBlend {blendOption=Linear}
    boneToModify {boneName=dyn_subthigh_l}
    scale {x=12, z=12}
   }
   n11 = "Transform (Modify) Bone - Bone: dyn_subthigh_r"(ComponentPose=@n10, Alpha=Get legSwingDelta_r() -> (legSwingDelta_r:double), Translation="0,0,0", Rotation="0,0,0", Scale="2.25,1,2.25") {
    scaleMode=BMM_Additive
    scaleSpace=BCS_BoneSpace
    translationMode=BMM_Additive
    translationSpace=BCS_BoneSpace
    alphaBoolBlend {blendOption=Linear}
    boneToModify {boneName=dyn_subthigh_r}
    scale {x=2.25, z=2.25}
   }
   n12 = "Constraint - dyn_gun_base (PPPPPP)"!(
    ComponentPose=@n11
    Alpha=1
    ConstraintWeights_0=0.0429
    ConstraintWeights_1:float=0.0605
    ConstraintWeights_2:float=0.001
    ConstraintWeights_3:float=0.0099
    ConstraintWeights_4:float=0.8701
    ConstraintWeights_5:float=0.0156
   ) -> (Pose) {
    alphaBoolBlend {blendOption=Linear}
    boneToModify {boneName=dyn_gun_base}
    constraintSetup [
     {targetBone {boneName=spine_01}}
     {targetBone {boneName=spine_02}}
     {targetBone {boneName=spine_03}}
     {targetBone {boneName=spine_04}}
     {targetBone {boneName=pelvis}}
     {targetBone {boneName=thigh_twist_01_l}}
    ]
    constraintWeights [1,1,1,1,1,1]
   }
   n13 = "Transform (Modify) Bone - Bone: dyn_gun_base"(ComponentPose=@n12, Alpha=1, Translation="0,0,0", Rotation="-8,0,0", Scale="1,1,1") {
    rotationMode=BMM_Additive
    rotationSpace=BCS_BoneSpace
    translationMode=BMM_Additive
    translationSpace=BCS_BoneSpace
    alphaBoolBlend {blendOption=Linear}
    alphaScaleBiasClamp {bInterpResult=true}
    boneToModify {boneName=dyn_gun_base}
    rotation {pitch=-8}
   }
   n14 = "Transform (Modify) Bone - Bone: dyn_gun_base"(ComponentPose=@n13, Alpha=Get thigh_rot_l(), Translation="-2,5,0", Rotation="0,-20,-30", Scale="1,1,1") {
    rotationMode=BMM_Additive
    rotationSpace=BCS_BoneSpace
    translationMode=BMM_Additive
    translationSpace=BCS_BoneSpace
    alphaBoolBlend {blendOption=Linear}
    alphaScaleBiasClamp {bInterpResult=true}
    boneToModify {boneName=dyn_gun_base}
    rotation {roll=-30, yaw=-20}
    translation=<-2 5>
   }
   Output Pose(Result=Component To Local(ComponentPose=@n14)) {layerGroup=None}
  ]}
 EventGraph {nodes [
   Event BlueprintUpdateAnimation()
   > n1 = Sequence()
   n2 = GetOwnerSkeletalMeshComponent()
   n3 = BreakRotator(InRot=BreakTransform(InTransform=GetDeltaTransformFromRefPose(self=@n2, BoneName=thigh_l, BaseName=pelvis)).Rotation)
   > Set thigh_rotation_l(thigh_rotation_l:double=MapRangeClamped(Value=@n3.Yaw, InRangeA=-50, InRangeB=-90)) -> (Output_Get:double)
   > Set thigh_rot_l(thigh_rot_l=MapRangeClamped(Value=@n3.Yaw, InRangeA=-55, InRangeB=-103))
   n4 = GetOwnerSkeletalMeshComponent()
   Set legSwingDelta_r(
    execute=@n1.then_1
    legSwingDelta_r:double=MapRangeClamped(Value=GetTwistAndSwingAngleOfDeltaRotationFromRefPose(self=@n4, BoneName=thigh_r).OutSwingAngle, InRangeA=150, InRangeB=75)
   ) -> (Output_Get:double)
   > Set legSwingDelta_l(legSwingDelta_l:double=MapRangeClamped(Value=GetTwistAndSwingAngleOfDeltaRotationFromRefPose(self=@n4, BoneName=thigh_l).OutSwingAngle, InRangeA=43, InRangeB=110)) -> (Output_Get:double)
  ]}
]
variables [
 legSwingDelta_l {type=number}
 legSwingDelta_r {type=number}
 thigh_rot_l {type=number}
 thigh_rotation_l {type=number}
]
