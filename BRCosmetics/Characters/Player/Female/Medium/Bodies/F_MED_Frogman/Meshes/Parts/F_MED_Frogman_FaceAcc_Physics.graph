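# Physics asset for the face accessory of the female medium Frogman player body, giving the facemask bones simulated secondary motion.
# Three box bodies sit on a small base bone, a parent bone and a tip bone, with the latter two boxes offset along the bone's X axis.
# Two constraints link them in a chain from base to parent to tip, each with moderate swing limits and a tight twist limit.
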PhysicsAsset /BRCosmetics/Characters/Player/Female/Medium/Bodies/F_MED_Frogman/Meshes/Parts/F_MED_Frogman_FaceAcc_Physics
bodies {
 dyn_facemask_01 [{shapeType=Box, extentX=3.1281, extentY=4.7398, extentZ=0.75, center=<1.4293 0 0>}]
 dyn_facemask_base [{shapeType=Box, extentX=0.5, extentY=0.5, extentZ=0.5}]
 dyn_facemask_parent [{shapeType=Box, extentX=2.26, extentY=4.01, extentZ=0.75, center=<0.8336 0 0>}]
}
constraints [
 {bone1Name=dyn_facemask_parent, bone2Name=dyn_facemask_base, swing1LimitDegrees=20, swing2LimitDegrees=20, twistLimitDegrees=5}
 {bone1Name=dyn_facemask_01, bone2Name=dyn_facemask_parent, swing1LimitDegrees=20, swing2LimitDegrees=20, twistLimitDegrees=5}
]
