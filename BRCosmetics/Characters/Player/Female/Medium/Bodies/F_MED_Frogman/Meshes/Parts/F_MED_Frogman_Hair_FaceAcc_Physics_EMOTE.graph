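# Physics asset for the hair of the medium female Frogman's hair and face-accessory part, apparently an emote-specific variant.
# Box bodies form six hair chains (back, back left, back right, left, right, front right), each constrained link by link to the head with per-joint swing and twist limits.
# Unconstrained capsule and sphere bodies on head, neck, upper spine, clavicles and right upper arm give the hair body shapes to collide with.
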PhysicsAsset /BRCosmetics/Characters/Player/Female/Medium/Bodies/F_MED_Frogman/Meshes/Parts/F_MED_Frogman_Hair_FaceAcc_Physics_EMOTE
bodies {
 clavicle_l [{shapeType=Capsule, length=7.2935, radius=7.3987, center=<5.6078 0.6445 -0.7283>, rotation=<rot 87.8752 0 0>}]
 clavicle_r [
  {shapeType=Capsule, length=9.6663, radius=8.3525, center=<-2.9614 -3.1063 0.391>, rotation=<rot 50.5521 -72.1584 -74.8911>}
  {shapeType=Capsule, length=2, radius=6.625, center=<-1.6314 1.1019 -4.5323>, rotation=<rot 0 0 15.4057>}
 ]
 dyn_hair_bk_01 [{shapeType=Box, extentX=5.9004, extentY=2.7856, extentZ=10.0124, center=<2.8612 0 0>}]
 dyn_hair_bk_02 [{shapeType=Box, extentX=5.9004, extentY=2.7856, extentZ=10.5124, center=<2.8123 0 0>}]
 dyn_hair_bk_03 [{shapeType=Box, extentX=8.9004, extentY=3.7856, extentZ=11.5124, center=<4.2228 0.3175 0>, rotation=<rot 0 8.6979 0>}]
 dyn_hair_bk_l_01 [{shapeType=Box, extentX=6.5295, extentY=2.8226, extentZ=7.5124, center=<3.3087 0 0>}]
 dyn_hair_bk_l_02 [{shapeType=Box, extentX=5.0295, extentY=1.5, extentZ=7.5124, center=<2.786 -0.4551 -0.0001>}]
 dyn_hair_bk_l_03 [{shapeType=Box, extentX=6.5295, extentY=2.8226, extentZ=7.5124, center=<3.3087 0 0>}]
 dyn_hair_bk_r_01 [{shapeType=Box, extentX=5.9004, extentY=2.6291, extentZ=7.5124, center=<3.3786 -0.0001 0>}]
 dyn_hair_bk_r_02 [{shapeType=Box, extentX=5.9004, extentY=2.6291, extentZ=7.5124, center=<2.7119 -0.3641 0>}]
 dyn_hair_bk_r_03 [{shapeType=Box, extentX=5.9004, extentY=2.6291, extentZ=7.5124, center=<2.7119 0 0>}]
 dyn_hair_fr_r_01 [{shapeType=Box, extentX=6.8593, extentY=2.041, extentZ=7.5124, center=<2.6177 0 -0.0001>}]
 dyn_hair_fr_r_02 [{shapeType=Box, extentX=8.9454, extentY=2.3327, extentZ=7.5266, center=<2.7932 0.481 -0.1787>, rotation=<rot -11.5076 2.4667 -3.1734>}]
 dyn_hair_l_01 [{shapeType=Box, extentX=5.9004, extentY=1.4486, extentZ=7.5124, center=<3.2348 0 -0.0001>}]
 dyn_hair_l_02 [{shapeType=Box, extentX=5.2096, extentY=1.8026, extentZ=7.5124, center=<3.2348 0 -0.0001>}]
 dyn_hair_l_03 [{shapeType=Box, extentX=2.2096, extentY=1.8791, extentZ=4.0124, center=<1.2769 0.0001 -0.0001>}]
 dyn_hair_r_01 [{shapeType=Box, extentX=5.2671, extentY=2.0549, extentZ=7.5124, center=<2.8597 0.0001 -0.0001>}]
 dyn_hair_r_02 [{shapeType=Box, extentX=6.3507, extentY=2.2833, extentZ=7.5124, center=<2.1272 -0.4815 -0.4758>, rotation=<rot 0.0279 -12.2612 0.0065>}]
 dyn_hair_r_03 [{shapeType=Box, extentX=2.4019, extentY=2.5459, extentZ=3.5124, center=<0.7435 -0.1448 -0.0001>, rotation=<rot 0 -10.5058 0>}]
 head [
  {shapeType=Sphere, radius=6.4177, center=<1.3282 -0.1209 -0.2823>}
  {shapeType=Capsule, length=5.9657, radius=5.522, center=<0.8118 1.1771 0>, rotation=<rot 90 0 0>}
  {
   shapeType=Capsule
   length=6.0522
   radius=7.194
   shapeName=head_capsule
   center=<-1.1124 4.6108 -0.4916>
   rotation=<rot -3.4221 -87.5597 89.8536>
  }
 ]
 neck_01 [{shapeType=Capsule, length=8.2568, radius=8.7685, center=<0 2.2205 -0.0001>, rotation=<rot 90 0 0>}]
 spine_05 [
  {shapeType=Capsule, length=0.9494, radius=8.7688, center=<7.8219 2.0809 0.8076>}
  {shapeType=Capsule, length=19.625, radius=10.375, center=<-5.3388 -0.7352 -3.0078>, rotation=<rot 90 -14.0363 -21.1593>}
 ]
 upperarm_r [{shapeType=Capsule, length=17, radius=5.5, center=<-8.5123 0 0>, rotation=<rot 90 0 0>}]
}
constraints [
 {bone1Name=dyn_hair_l_01, bone2Name=head, swing1LimitDegrees=30, swing2LimitDegrees=20, twistLimitDegrees=8}
 {bone1Name=dyn_hair_l_02, bone2Name=dyn_hair_l_01, swing1LimitDegrees=30, swing2LimitDegrees=20, twistLimitDegrees=10}
 {bone1Name=dyn_hair_bk_r_01, bone2Name=head, swing1LimitDegrees=60, swing2LimitDegrees=20, twistLimitDegrees=20}
 {bone1Name=dyn_hair_bk_r_02, bone2Name=dyn_hair_bk_r_01, swing1LimitDegrees=30, swing2LimitDegrees=20, twistLimitDegrees=10}
 {bone1Name=dyn_hair_bk_r_03, bone2Name=dyn_hair_bk_r_02, swing1LimitDegrees=30, swing2LimitDegrees=20, twistLimitDegrees=10}
 {bone1Name=dyn_hair_bk_01, bone2Name=head, swing1LimitDegrees=40, swing2LimitDegrees=12, twistLimitDegrees=10}
 {bone1Name=dyn_hair_bk_02, bone2Name=dyn_hair_bk_01, swing1LimitDegrees=60, swing2LimitDegrees=15, twistLimitDegrees=8}
 {bone1Name=dyn_hair_bk_03, bone2Name=dyn_hair_bk_02, swing1LimitDegrees=50, swing2LimitDegrees=15, twistLimitDegrees=8}
 {bone1Name=dyn_hair_bk_l_01, bone2Name=head, swing1LimitDegrees=45, swing2LimitDegrees=20, twistLimitDegrees=5}
 {bone1Name=dyn_hair_bk_l_02, bone2Name=dyn_hair_bk_l_01, swing1LimitDegrees=30, swing2LimitDegrees=20, twistLimitDegrees=8}
 {bone1Name=dyn_hair_bk_l_03, bone2Name=dyn_hair_bk_l_02, swing1LimitDegrees=35, swing2LimitDegrees=20, twistLimitDegrees=10}
 {bone1Name=dyn_hair_r_01, bone2Name=head, swing1LimitDegrees=50, swing2LimitDegrees=40, twistLimitDegrees=10}
 {bone1Name=dyn_hair_r_02, bone2Name=dyn_hair_r_01, swing1LimitDegrees=40, swing2LimitDegrees=35, twistLimitDegrees=10}
 {bone1Name=dyn_hair_r_03, bone2Name=dyn_hair_r_02, swing1LimitDegrees=20, swing2LimitDegrees=20, twistLimitDegrees=20}
 {bone1Name=dyn_hair_fr_r_01, bone2Name=head, swing1LimitDegrees=30, swing2LimitDegrees=30, twistLimitDegrees=10}
 {bone1Name=dyn_hair_fr_r_02, bone2Name=dyn_hair_fr_r_01, swing1LimitDegrees=20, swing2LimitDegrees=30, twistLimitDegrees=20}
 {bone1Name=dyn_hair_l_03, bone2Name=dyn_hair_l_02, swing1LimitDegrees=20, swing2LimitDegrees=20, twistLimitDegrees=12}
]
