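# Post-process style Anim Blueprint for the female Frogman face accessory, simulating and correcting hair over the pose copied from the parent mesh.
# Skydive dive mode blends to a separate RigidBody setup with fixed bone offsets.
# Otherwise, hair and accessory RigidBody passes, plus an emote pass gated by frame delta time, feed additive bone corrections.
# EventGraph drives those corrections from weapon, aim, crouch, air state, head rotation and right-arm twist.
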AnimBlueprint /BRCosmetics/Characters/Player/Female/Medium/Bodies/F_MED_Frogman/Meshes/Parts/F_MED_Frogman_FaceAcc_AnimBp
target_skeleton=&/F_MED_Frogman_FaceAcc_Skeleton
graphs [
 AnimGraph {nodes [
   "Transform (Modify) Bone - Bone: dyn_hair_bk_01"(Alpha=1, Translation="0,-1,0", Rotation="0,-3,0", Scale="1,1,1") {
    rotationMode=BMM_Additive
    rotationSpace=BCS_BoneSpace
    scaleMode=BMM_Additive
    scaleSpace=BCS_BoneSpace
    translationMode=BMM_Additive
    translationSpace=BCS_BoneSpace
    alphaBoolBlend {blendInTime=0.25, blendOption=Linear, blendOutTime=0.25}
    alphaScaleBiasClamp {bInterpResult=true, interpSpeedDecreasing=8, interpSpeedIncreasing=8}
    boneToModify {boneName=dyn_hair_bk_01}
    rotation {yaw=-3}
    translation {y=-1}
   }
   n1 = RigidBody!(
    ComponentPose=Local To Component(LocalPose=Copy Pose From Mesh() {bUseAttachedParent=true})
    Alpha
    OverrideWorldGravity:Vector="0,0,-980"
    ExternalForce
   ) -> (Pose) {
    bClampLinearTranslationLimitToRefPose=true
    bOverrideWorldGravity=true
    overridePhysicsAsset=&/F_MED_Frogman_Hair_DiveMode_FaceAcc_Physics
    simulationSpace=BaseBoneSpace
    alphaBoolBlend {blendOption=Linear}
    alphaScaleBiasClamp {interpSpeedDecreasing=4, interpSpeedIncreasing=4}
    baseBoneRef {boneName=spine_03}
    componentAppliedLinearAccClamp=<600 600 600>
    componentLinearAccScale=<0.35 0.35 0.35>
    componentLinearVelScale=<0.35 0.35 0.35>
    overrideWorldGravity {z=-980}
   }
   n2 = "Transform (Modify) Bone - Bone: dyn_hair_bk_01"(ComponentPose=@n1, Alpha=1, Translation="0,0,0", Rotation="0,-10,0", Scale="1,1,1") {
    rotationMode=BMM_Additive
    rotationSpace=BCS_BoneSpace
    translationMode=BMM_Additive
    translationSpace=BCS_BoneSpace
    alphaBoolBlend {blendOption=Linear}
    boneToModify {boneName=dyn_hair_bk_01}
    rotation {yaw=-10}
   }
   n3 = "Transform (Modify) Bone - Bone: dyn_hair_l_01"(ComponentPose=@n2, Alpha=1, Translation="0,0,0", Rotation="-15,0,0", Scale="1,1,1") {
    rotationMode=BMM_Additive
    rotationSpace=BCS_BoneSpace
    translationMode=BMM_Additive
    translationSpace=BCS_BoneSpace
    alphaBoolBlend {blendOption=Linear}
    boneToModify {boneName=dyn_hair_l_01}
    rotation {pitch=-15}
   }
   n4 = "Transform (Modify) Bone - Bone: dyn_hair_r_01"(ComponentPose=@n3, Alpha=1, Translation="0,0,0", Rotation="15,0,0", Scale="1,1,1") {
    rotationMode=BMM_Additive
    rotationSpace=BCS_BoneSpace
    translationMode=BMM_Additive
    translationSpace=BCS_BoneSpace
    alphaBoolBlend {blendOption=Linear}
    boneToModify {boneName=dyn_hair_r_01}
    rotation {pitch=15}
   }
   n5 = RigidBody!(ComponentPose=Local To Component(LocalPose=Copy Pose From Mesh() {bUseAttachedParent=true}), Alpha, ExternalForce) -> (Pose) {
    bClampLinearTranslationLimitToRefPose=true
    bOverrideWorldGravity=true
    overridePhysicsAsset=&/F_MED_Frogman_Hair_FaceAcc_Physics
    simulationSpace=BaseBoneSpace
    alphaBoolBlend {blendOption=Linear}
    alphaScaleBiasClamp {interpSpeedDecreasing=4, interpSpeedIncreasing=4}
    baseBoneRef {boneName=spine_03}
    componentAppliedLinearAccClamp=<600 600 600>
    componentLinearAccScale=<0.35 0.35 0.35>
    componentLinearVelScale=<0.35 0.35 0.35>
    overrideWorldGravity {z=-980}
   }
   n6 = RigidBody(ComponentPose=@n5) {
    bClampLinearTranslationLimitToRefPose=true
    bOverrideWorldGravity=true
    overridePhysicsAsset=&/F_MED_Frogman_FaceAcc_Physics
    simulationSpace=BaseBoneSpace
    alphaBoolBlend {blendOption=Linear}
    baseBoneRef {boneName=root}
    componentAppliedLinearAccClamp=<6000 6000 6000>
    componentLinearAccScale=<0.25 0.25 0.25>
    componentLinearVelScale=<0.25 0.25 0.25>
    overrideWorldGravity {z=-980}
   }
   n7 = RigidBody(ComponentPose=@n6, Alpha=Get fr_hairstrand_fix() -> (fr_hairstrand_fix:double)) {
    bClampLinearTranslationLimitToRefPose=true
    bOverrideWorldGravity=true
    overridePhysicsAsset=&/F_MED_Frogman_Hair_FaceAcc_Physics_EMOTE
    simulationSpace=BaseBoneSpace
    alphaBoolBlend {blendOption=Linear}
    alphaScaleBiasClamp {interpSpeedDecreasing=4, interpSpeedIncreasing=4}
    baseBoneRef {boneName=spine_03}
    componentAppliedLinearAccClamp=<600 600 600>
    componentLinearAccScale=<0.35 0.35 0.35>
    componentLinearVelScale=<0.35 0.35 0.35>
    overrideWorldGravity {z=-980}
   }
   n8 = "Transform (Modify) Bone - Bone: dyn_hair_bk_r_01"(ComponentPose=@n7, Alpha=Get ar_firingPosition() -> (ar_firingPosition:double), Translation="0,0,0", Rotation="0,-15,0", Scale="1,1,1") {
    rotationMode=BMM_Additive
    rotationSpace=BCS_BoneSpace
    scaleMode=BMM_Additive
    scaleSpace=BCS_BoneSpace
    translationMode=BMM_Additive
    translationSpace=BCS_BoneSpace
    alphaBoolBlend {blendInTime=0.25, blendOption=Linear, blendOutTime=0.25}
    alphaScaleBiasClamp {bInterpResult=true, interpSpeedDecreasing=8, interpSpeedIncreasing=8}
    boneToModify {boneName=dyn_hair_bk_r_01}
    rotation {yaw=-15}
   }
   n9 = "Transform (Modify) Bone - Bone: dyn_hair_bk_r_01"(
    ComponentPose=@n8
    Alpha=Get ar_firingPosition_moving() -> (ar_firingPosition_moving:double)
    Translation="0,0,0"
    Rotation="20,-50,0"
    Scale="1,1,1"
   ) {
    rotationMode=BMM_Additive
    rotationSpace=BCS_BoneSpace
    scaleMode=BMM_Additive
    scaleSpace=BCS_BoneSpace
    translationMode=BMM_Additive
    translationSpace=BCS_BoneSpace
    alphaBoolBlend {blendInTime=0.25, blendOption=Linear, blendOutTime=0.25}
    alphaScaleBiasClamp {bInterpResult=true, interpSpeedDecreasing=8, interpSpeedIncreasing=8}
    boneToModify {boneName=dyn_hair_bk_r_01}
    rotation {pitch=20, yaw=-50}
   }
   n10 = "Transform (Modify) Bone - Bone: dyn_hair_bk_r_02"(ComponentPose=@n9, Alpha=Get ADS_Alpha() -> (ADS_Alpha:double), Translation="0,0,0", Rotation="0,-20,0", Scale="1,1,1") {
    rotationMode=BMM_Additive
    rotationSpace=BCS_BoneSpace
    translationMode=BMM_Additive
    translationSpace=BCS_BoneSpace
    alphaBoolBlend {blendOption=Linear}
    alphaScaleBiasClamp {bInterpResult=true, interpSpeedDecreasing=8, interpSpeedIncreasing=8}
    boneToModify {boneName=dyn_hair_bk_r_02}
    rotation {yaw=-20}
   }
   n11 = NOT Boolean(A="float > float"(A=Get PawnSpeed2D(), B=1) {is=title K2Node_CallFunction})
   n12 = Get bIsTargeting()
   n13 = "Transform (Modify) Bone - Bone: dyn_hair_bk_r_02"(
    ComponentPose=@n10
    Alpha=SelectFloat(A=1, B=0, bPickA=AND Boolean(A=@n12, B=@n11))
    Translation="-1,0,0"
    Rotation="-25,25,-25"
    Scale="1,1,1"
   ) {
    rotationMode=BMM_Additive
    rotationSpace=BCS_BoneSpace
    scaleMode=BMM_Replace
    scaleSpace=BCS_BoneSpace
    translationMode=BMM_Additive
    translationSpace=BCS_BoneSpace
    alphaBoolBlend {blendOption=Linear}
    alphaScaleBiasClamp {bInterpResult=true, interpSpeedDecreasing=8, interpSpeedIncreasing=8}
    boneToModify {boneName=dyn_hair_bk_r_02}
    rotation=<rot -25 25 -25>
    translation {x=-1}
   }
   n14 = Get CrouchADS_Moving_Alpha() -> (CrouchADS_Moving_Alpha:double)
   n15 = "Transform (Modify) Bone - Bone: dyn_hair_r_02"(ComponentPose=@n13, Alpha=@n14, Translation="0,-2,0", Rotation="0,20,0", Scale="1,1,1") {
    rotationMode=BMM_Additive
    rotationSpace=BCS_BoneSpace
    scaleMode=BMM_Additive
    scaleSpace=BCS_BoneSpace
    translationMode=BMM_Additive
    translationSpace=BCS_BoneSpace
    alphaBoolBlend {blendInTime=0.25, blendOption=Linear, blendOutTime=0.25}
    alphaScaleBiasClamp {bInterpResult=true, interpSpeedDecreasing=8, interpSpeedIncreasing=8}
    boneToModify {boneName=dyn_hair_r_02}
    rotation {yaw=20}
    translation {y=-2}
   }
   n16 = "Transform (Modify) Bone - Bone: dyn_hair_bk_r_01"(ComponentPose=@n15, Alpha=@n14, Translation="0,0,0", Rotation="40,15,0", Scale="1,1,1") {
    rotationMode=BMM_Additive
    rotationSpace=BCS_BoneSpace
    scaleMode=BMM_Additive
    scaleSpace=BCS_BoneSpace
    translationMode=BMM_Additive
    translationSpace=BCS_BoneSpace
    alphaBoolBlend {blendInTime=0.25, blendOption=Linear, blendOutTime=0.25}
    alphaScaleBiasClamp {bInterpResult=true, interpSpeedDecreasing=8, interpSpeedIncreasing=8}
    boneToModify {boneName=dyn_hair_bk_r_01}
    rotation {pitch=40, yaw=15}
   }
   n17 = "Transform (Modify) Bone - Bone: dyn_hair_bk_r_02"(ComponentPose=@n16, Alpha=@n14, Translation="0,0,0", Rotation="-50,13,0", Scale="1,1,1") {
    rotationMode=BMM_Additive
    rotationSpace=BCS_BoneSpace
    scaleMode=BMM_Additive
    scaleSpace=BCS_BoneSpace
    translationMode=BMM_Additive
    translationSpace=BCS_BoneSpace
    alphaBoolBlend {blendInTime=0.25, blendOption=Linear, blendOutTime=0.25}
    alphaScaleBiasClamp {bInterpResult=true, interpSpeedDecreasing=8, interpSpeedIncreasing=8}
    boneToModify {boneName=dyn_hair_bk_r_02}
    rotation {pitch=-50, yaw=13}
   }
   n18 = "Transform (Modify) Bone - Bone: dyn_hair_bk_r_01"(ComponentPose=@n17, Alpha=Get head_rot_back() -> (head_rot_back:double), Translation="0,0,0", Rotation="0,-5,0", Scale="1,1,1") {
    rotationMode=BMM_Additive
    rotationSpace=BCS_BoneSpace
    scaleMode=BMM_Replace
    scaleSpace=BCS_BoneSpace
    translationMode=BMM_Additive
    translationSpace=BCS_BoneSpace
    alphaBoolBlend {blendOption=Linear}
    alphaScaleBiasClamp {bInterpResult=true, interpSpeedDecreasing=8, interpSpeedIncreasing=8}
    boneToModify {boneName=dyn_hair_bk_r_01}
    rotation {yaw=-5}
   }
   n19 = "Transform (Modify) Bone - Bone: dyn_hair_r_01"(ComponentPose=@n18, Alpha=Get ADS_Alpha(), Translation="0,0,0", Rotation="15,10,0", Scale="1,1,1") {
    rotationMode=BMM_Additive
    rotationSpace=BCS_BoneSpace
    translationMode=BMM_Additive
    translationSpace=BCS_BoneSpace
    alphaBoolBlend {blendOption=Linear}
    alphaScaleBiasClamp {bInterpResult=true, interpSpeedDecreasing=8, interpSpeedIncreasing=8}
    boneToModify {boneName=dyn_hair_r_01}
    rotation {pitch=15, yaw=10}
   }
   n20 = "Transform (Modify) Bone - Bone: dyn_hair_bk_r_02"(ComponentPose=@n19, Alpha=Get shldr_r_alpha() -> (shldr_r_alpha:double), Translation="0,0,0", Rotation="20,-20,5", Scale="1,1,1") {
    rotationMode=BMM_Additive
    rotationSpace=BCS_BoneSpace
    scaleMode=BMM_Additive
    scaleSpace=BCS_BoneSpace
    translationMode=BMM_Additive
    translationSpace=BCS_BoneSpace
    alphaBoolBlend {blendOption=Linear}
    alphaScaleBiasClamp {bInterpResult=true, interpSpeedDecreasing=8, interpSpeedIncreasing=8}
    boneToModify {boneName=dyn_hair_bk_r_02}
    rotation=<rot 20 -20 5>
   }
   n21 = "Transform (Modify) Bone - Bone: dyn_hair_bk_02"(ComponentPose=@n20, Alpha=1, Translation="0,0,0", Rotation="0,0,20", Scale="1,1,1") {
    rotationMode=BMM_Additive
    rotationSpace=BCS_BoneSpace
    translationMode=BMM_Additive
    translationSpace=BCS_BoneSpace
    alphaBoolBlend {blendOption=Linear}
    alphaScaleBiasClamp {bInterpResult=true, interpSpeedDecreasing=8, interpSpeedIncreasing=8}
    boneToModify {boneName=dyn_hair_bk_02}
    rotation {roll=20}
   }
   n22 = Get head_rot_back()
   n23 = "Transform (Modify) Bone - Bone: dyn_hair_bk_r_01"(ComponentPose=@n21, Alpha=@n22, Translation="-2,-2,0", Rotation="0,0,0", Scale="1,1,1") {
    rotationMode=BMM_Additive
    rotationSpace=BCS_BoneSpace
    translationMode=BMM_Additive
    translationSpace=BCS_BoneSpace
    alphaBoolBlend {blendOption=Linear}
    alphaScaleBiasClamp {bInterpResult=true, interpSpeedDecreasing=8, interpSpeedIncreasing=8}
    boneToModify {boneName=dyn_hair_bk_r_01}
    translation=<-2 -2>
   }
   n24 = "Transform (Modify) Bone - Bone: dyn_hair_bk_r_02"(ComponentPose=@n23, Alpha=@n22, Translation="-2,-2,2", Rotation="-20,-30,0", Scale="1,1,1") {
    rotationMode=BMM_Additive
    rotationSpace=BCS_BoneSpace
    translationMode=BMM_Additive
    translationSpace=BCS_BoneSpace
    alphaBoolBlend {blendOption=Linear}
    alphaScaleBiasClamp {bInterpResult=true, interpSpeedDecreasing=8, interpSpeedIncreasing=8}
    boneToModify {boneName=dyn_hair_bk_r_02}
    rotation {pitch=-20, yaw=-30}
    translation=<-2 -2 2>
   }
   n25 = "Transform (Modify) Bone - Bone: dyn_hair_bk_02"(ComponentPose=@n24, Alpha=@n22, Translation="0,0,0", Rotation="0,-15,0", Scale="1,1,1") {
    rotationMode=BMM_Additive
    rotationSpace=BCS_BoneSpace
    translationMode=BMM_Additive
    translationSpace=BCS_BoneSpace
    alphaBoolBlend {blendOption=Linear}
    alphaScaleBiasClamp {bInterpResult=true, interpSpeedDecreasing=8, interpSpeedIncreasing=8}
    boneToModify {boneName=dyn_hair_bk_02}
    rotation {yaw=-15}
   }
   n26 = Get shldr_r_alpha()
   n27 = "Transform (Modify) Bone - Bone: dyn_hair_fr_r_01"(ComponentPose=@n25, Alpha=@n26, Translation="0,-1,1", Rotation="0,20,0", Scale="1,1,1") {
    rotationMode=BMM_Additive
    rotationSpace=BCS_BoneSpace
    translationMode=BMM_Additive
    translationSpace=BCS_BoneSpace
    alphaBoolBlend {blendOption=Linear}
    alphaScaleBiasClamp {bInterpResult=true, interpSpeedDecreasing=8, interpSpeedIncreasing=8}
    boneToModify {boneName=dyn_hair_fr_r_01}
    rotation {yaw=20}
    translation {y=-1, z=1}
   }
   n28 = "Transform (Modify) Bone - Bone: dyn_hair_bk_r_01"(ComponentPose=@n27, Alpha=@n26, Translation="0,-2,0", Rotation="0,0,0", Scale="1,1,1") {
    rotationMode=BMM_Additive
    rotationSpace=BCS_BoneSpace
    translationMode=BMM_Additive
    translationSpace=BCS_BoneSpace
    alphaBoolBlend {blendOption=Linear}
    alphaScaleBiasClamp {bInterpResult=true, interpSpeedDecreasing=8, interpSpeedIncreasing=8}
    boneToModify {boneName=dyn_hair_bk_r_01}
    translation {y=-2}
   }
   n29 = "Transform (Modify) Bone - Bone: dyn_hair_r_01"(ComponentPose=@n28, Alpha=@n26, Translation="0,0,0", Rotation="0,-20,0", Scale="1,1,1") {
    rotationMode=BMM_Additive
    rotationSpace=BCS_BoneSpace
    translationMode=BMM_Additive
    translationSpace=BCS_BoneSpace
    alphaBoolBlend {blendOption=Linear}
    alphaScaleBiasClamp {bInterpResult=true, interpSpeedDecreasing=8, interpSpeedIncreasing=8}
    boneToModify {boneName=dyn_hair_r_01}
    rotation {yaw=-20}
   }
   n30 = Get Skydive_Alpha() -> (Skydive_Alpha:double)
   n31 = "Transform (Modify) Bone - Bone: dyn_hair_bk_r_01"(ComponentPose=@n29, Alpha=@n30, Translation="0,0,0", Rotation="0,-30,0", Scale="1,1,1") {
    rotationMode=BMM_Additive
    rotationSpace=BCS_BoneSpace
    translationMode=BMM_Additive
    translationSpace=BCS_BoneSpace
    alphaBoolBlend {blendOption=Linear}
    alphaScaleBiasClamp {bInterpResult=true, interpSpeedDecreasing=8, interpSpeedIncreasing=8}
    boneToModify {boneName=dyn_hair_bk_r_01}
    rotation {yaw=-30}
   }
   n32 = "Transform (Modify) Bone - Bone: dyn_hair_bk_l_01"(ComponentPose=@n31, Alpha=@n30, Translation="0,0,0", Rotation="-30,-60,0", Scale="1,1,1") {
    rotationMode=BMM_Additive
    rotationSpace=BCS_BoneSpace
    translationMode=BMM_Additive
    translationSpace=BCS_BoneSpace
    alphaBoolBlend {blendOption=Linear}
    alphaScaleBiasClamp {bInterpResult=true, interpSpeedDecreasing=8, interpSpeedIncreasing=8}
    boneToModify {boneName=dyn_hair_bk_l_01}
    rotation {pitch=-30, yaw=-60}
   }
   n33 = Get Falling_Alpha() -> (Falling_Alpha:double)
   n34 = "Transform (Modify) Bone - Bone: dyn_hair_bk_r_01"(ComponentPose=@n32, Alpha=@n33, Translation="0,0,0", Rotation="-10,-10,0", Scale="1,1,1") {
    rotationMode=BMM_Additive
    rotationSpace=BCS_BoneSpace
    translationMode=BMM_Additive
    translationSpace=BCS_BoneSpace
    alphaBoolBlend {blendOption=Linear}
    alphaScaleBiasClamp {bInterpResult=true, interpSpeedDecreasing=8, interpSpeedIncreasing=8}
    boneToModify {boneName=dyn_hair_bk_r_01}
    rotation {pitch=-10, yaw=-10}
   }
   n35 = "Transform (Modify) Bone - Bone: dyn_hair_bk_02"(ComponentPose=@n34, Alpha=@n33, Translation="0,0,0", Rotation="0,-25,0", Scale="1,1,1") {
    rotationMode=BMM_Additive
    rotationSpace=BCS_BoneSpace
    translationMode=BMM_Additive
    translationSpace=BCS_BoneSpace
    alphaBoolBlend {blendOption=Linear}
    alphaScaleBiasClamp {bInterpResult=true, interpSpeedDecreasing=8, interpSpeedIncreasing=8}
    boneToModify {boneName=dyn_hair_bk_02}
    rotation {yaw=-25}
   }
   n36 = Get Is_Jumping() -> (Is_Jumping:double)
   n37 = "Transform (Modify) Bone - Bone: dyn_hair_bk_r_01"(ComponentPose=@n35, Alpha=@n36, Translation="0,-2,0", Rotation="-10,-10,0", Scale="1,1,1") {
    rotationMode=BMM_Additive
    rotationSpace=BCS_BoneSpace
    translationMode=BMM_Additive
    translationSpace=BCS_BoneSpace
    alphaBoolBlend {blendOption=Linear}
    alphaScaleBiasClamp {bInterpResult=true, interpSpeedDecreasing=8, interpSpeedIncreasing=8}
    boneToModify {boneName=dyn_hair_bk_r_01}
    rotation {pitch=-10, yaw=-10}
    translation {y=-2}
   }
   n38 = "Transform (Modify) Bone - Bone: dyn_hair_bk_02"(ComponentPose=@n37, Alpha=@n36, Translation="0,0,0", Rotation="0,-5,0", Scale="1,1,1") {
    rotationMode=BMM_Additive
    rotationSpace=BCS_BoneSpace
    translationMode=BMM_Additive
    translationSpace=BCS_BoneSpace
    alphaBoolBlend {blendOption=Linear}
    alphaScaleBiasClamp {bInterpResult=true, interpSpeedDecreasing=8, interpSpeedIncreasing=8}
    boneToModify {boneName=dyn_hair_bk_02}
    rotation {yaw=-5}
   }
   n39 = NOT Boolean(A=Get bIsTargeting())
   n40 = "float > float"(A=Get PawnSpeed2D(), B=1)
   n41 = Get bIsCrouching()
   n42 = "Transform (Modify) Bone - Bone: dyn_hair_bk_l_01"(
    ComponentPose=@n38
    Alpha=SelectFloat(A=1, B=0.5, bPickA=AND Boolean!(A=@n41, B=@n40, C:bool=@n39) -> (ReturnValue))
    Translation="0,0,0"
    Rotation="-25,-35,0"
    Scale="1,1,1"
   ) {
    rotationMode=BMM_Additive
    rotationSpace=BCS_BoneSpace
    translationMode=BMM_Additive
    translationSpace=BCS_BoneSpace
    alphaBoolBlend {blendOption=Linear}
    alphaScaleBiasClamp {bInterpResult=true, interpSpeedDecreasing=8, interpSpeedIncreasing=8}
    boneToModify {boneName=dyn_hair_bk_l_01}
    rotation {pitch=-25, yaw=-35}
   }
   Output Pose(Result=Blend Poses by bool(
     BlendPose_0=Component To Local(ComponentPose=@n4)
     BlendPose_1=Component To Local(ComponentPose=@n42)
     BlendTime_0=0.5
     BlendTime_1=1
     bActiveValue=Get bIsSkydiveDiveMode()
    ) {blendType=Linear}) {layerGroup=None}
  ]}
 EventGraph {nodes [
   Event BlueprintInitializeAnimation()
   > n1 = Cast To PlayerPawn_Athena(Object=TryGetPawnOwner())
   > n2 = Set Pawn Owner(Pawn Owner:Player Pawn Athena Object Reference=@n1) -> (Output_Get:Player Pawn Athena Object Reference)
   > Set Body_Mesh(Body_Mesh=GetSkeletalMeshForPartType(self=@n2))
   n3 = Event BlueprintUpdateAnimation()
   > n4 = Sequence!(execute) -> (then_0, then_1, then_2:Exec, then_3:Exec, then_4:Exec, then_5:Exec, then_6:Exec, then_7:Exec, then_8:Exec, then_9:Exec)
   n5 = Cast To FortPlayerAnimInstance(execute=[@n4.then_0, @n4.then_6], Object=GetAnimInstance(self=GetOwnerSkeletalMeshComponent()))
   > Set ar_firingPosition(ar_firingPosition:double=0) -> (Output_Get:double)
   > Set ar_firingPosition_moving(ar_firingPosition_moving:double=0) -> (Output_Get:double)
   > n6 = Switch on EFortWeaponCoreAnimation!(execute, Selection:EFortWeaponCoreAnimation Enum=Get WeaponCoreAnim(self=@n5)) -> (
    Melee:Exec
    Pistol:Exec
    Shotgun:Exec
    PaperBlueprint:Exec
    Rifle:Exec
    MeleeOneHand:Exec
    MachinePistol:Exec
    RocketLauncher:Exec
    GrenadeLauncher:Exec
    GoingCommando:Exec
    AssaultRifle:Exec
    TacticalShotgun:Exec
    SniperRifle:Exec
    TrapPlacement:Exec
    ShoulderLauncher:Exec
    AbilityDecoTool:Exec
    Crossbow:Exec
    C4:Exec
    RemoteControl:Exec
    DualWield:Exec
    AR_BullPup:Exec
    AR_ForwardGrip:Exec
    MedPackPaddles:Exec
    SMG_P90:Exec
    AR_DrumGun:Exec
    Consumable_Small:Exec
    Consumable_Large:Exec
    Balloon:Exec
    MountedTurret:Exec
    CreativeTool:Exec
    ExplosiveBow:Exec
    AshtonIndigo:Exec
    AshtonChicago:Exec
    MeleeDualWield:Exec
    Unarmed:Exec
   )
   n7 = Branch(execute=[@n6.AR_DrumGun, @n6.GrenadeLauncher, @n6.SMG_P90], Condition=Get bIsTargeting())
   Set ar_firingPosition(execute=[@n12.else, @n7.then, @n10.then, @n6.RocketLauncher], ar_firingPosition=0)
   n8 = Get PawnSpeed2D()
   > n9 = Branch(Condition=AND Boolean!(A="float > float"(A=@n8, B=1), B=Get bIsCrouching()) -> (ReturnValue))
   > Set ar_firingPosition_moving(ar_firingPosition_moving=1)
   Set ar_firingPosition(execute=@n9.else, ar_firingPosition=1)
   > Set ar_firingPosition_moving(ar_firingPosition_moving=0)
   Set ar_firingPosition(execute=[@n11.then, @n12.then, @n7.else], ar_firingPosition=0)
   n10 = Branch(execute=[@n6.AssaultRifle, @n6.Rifle, @n6.Shotgun, @n6.SniperRifle, @n6.TacticalShotgun], Condition=Get bIsTargeting())
   n11 = Branch(execute=@n10.else, Condition=Get bIsRelaxedLevel1() -> (bIsRelaxedLevel1:bool))
   n12 = Branch(execute=@n11.else, Condition=Get bIsRelaxedLevel2() -> (bIsRelaxedLevel2:bool))
   Set ADS_Alpha(execute=@n4.then_7, ADS_Alpha:double=SelectFloat(A=1, B=0, bPickA=Get bIsTargeting())) -> (Output_Get:double)
   n13 = "In Range (Float)"!(Value:double=@n3.DeltaTimeX, Min:double=0.016, Max:double=0.099, InclusiveMin:bool=true, InclusiveMax:bool=true) -> (ReturnValue:bool)
   Set fr_hairstrand_fix(execute=@n4.then_8, fr_hairstrand_fix:double=SelectFloat(A="To Float (Boolean)"(InBool=@n13), B=0, bPickA=Get bIsPlayingEmote())) -> (Output_Get:double)
   n14 = BreakTransform(InTransform=GetDeltaTransformFromRefPose(self=GetOwnerSkeletalMeshComponent(), BoneName=head, BaseName=spine_05))
   Set head_rot_back(execute=@n4.then_1, head_rot_back:double=MapRangeClamped(Value=BreakRotator(InRot=@n14.Rotation).Yaw, InRangeA=0, InRangeB=-48)) -> (Output_Get:double)
   n15 = Get bIsSkydiveFloating()
   n16 = Get bIsSkydiveDiveMode()
   Set Skydive_Alpha(execute=@n4.then_3, Skydive_Alpha:double=SelectFloat(A=1, B=0, bPickA=OR Boolean(A=@n16, B=@n15))) -> (Output_Get:double)
   > Set Falling_Alpha(Falling_Alpha:double=SelectFloat(A=1, B=0, bPickA=Get bIsFalling())) -> (Output_Get:double)
   > Set Is_Jumping(Is_Jumping:double=SelectFloat(A=1, B=0, bPickA=Get bIsJumping())) -> (Output_Get:double)
   n17 = Get bIsTargeting()
   n18 = Get bIsCrouching()
   Set CrouchADS_Alpha(execute=@n4.then_4, CrouchADS_Alpha:double=SelectFloat(A=1, B=0, bPickA=AND Boolean(A=@n17, B=@n18))) -> (Output_Get:double)
   n19 = "float > float"(A=Get PawnSpeed2D(), B=1)
   n20 = Get bIsTargeting()
   n21 = Get bIsCrouching()
   Set CrouchADS_Moving_Alpha(execute=@n4.then_5, CrouchADS_Moving_Alpha:double=SelectFloat(A=1, B=0, bPickA=AND Boolean!(A=@n20, B=@n21, C:bool=@n19) -> (ReturnValue))) -> (Output_Get:double)
   Set shldr_r_alpha(
    execute=@n4.then_2
    shldr_r_alpha:double=GetAlphaFromReferenceTransformDelta(
     SkeletalMeshComponent=Get Body_Mesh()
     BoneName=upperarm_r
     BaseName=clavicle_r
     DeltaComponentType=QuaternionTwist
     InputRange=(X=-81,Y=-120)
     TwistAxis="0,0,1"
    )
   ) -> (Output_Get:double)
   n22 = BreakTransform(InTransform=GetDeltaTransformFromRefPose(self=GetOwnerSkeletalMeshComponent(), BoneName=head, BaseName=spine_05))
   > Set head_rot_back(head_rot_back=MapRangeClamped(Value=BreakRotator(InRot=@n22.Rotation).Yaw, InRangeA=0, InRangeB=-48))
   n23 = Gravity Override State Function DEPRECATED(
    execute=@n4.then_9
    CharacterPartAnimInstance=Self-Reference()
    GameTimeInSeconds=GetGameTimeInSeconds()
    GravityOverrideParams=Get Gravity Override Params()
   )
   > Set GravityOverride(GravityOverride=@n23)
   TryGetPawnOwner()
  ]}
]
variables [
 ADS_Alpha {type=number}
 ar_firingPosition {type=number}
 ar_firingPosition_moving {type=number}
 Body_Mesh {type=SkeletalMeshComponent}
 CrouchADS_Alpha {type=number}
 CrouchADS_Moving_Alpha {type=number}
 Falling_Alpha {type=number}
 fr_hairstrand_fix {type=number}
 Gravity Override Params {type=GravityOverrideParamsStruct}
 GravityOverride {type=Vector}
 head_rot_back {type=number}
 Is_Jumping {type=number}
 Pawn Owner {type=PlayerPawn_Athena_C}
 shldr_r_alpha {type=number}
 Skydive_Alpha {type=number}
]
